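; Landmass component that builds a 2D distance field from a seed render target using GPU jump flooding.
; FindEdges draws seed edges, with optional depth and curl parameters, into float render targets RTA and RTB, allocated on demand.
; Blur and jump steps ping-pong between those targets through dynamic material instances, using pass counters to track source and target.
; ComputeDistance draws the final result into a single-channel distance field target.
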(event Custom|Assign%20RenderTargets
  (params (RTA :type Texture%20Render%20Target%202D%20Object%20Reference) (RTB :type Texture%20Render%20Target%202D%20Object%20Reference))
  (Variables|Default|SetRTA RTA)
  (Variables|Default|SetRTB RTB))

(event Custom|JumpFlood
  (params (SeedRT :type Texture%20Render%20Target%202D%20Object%20Reference) (SceneCaptureZ :type Float%20%28double-precision%29) (Curl :type Linear%20Color%20Structure) (UseDepth :type Boolean) (Z%20Location :type Float%20%28double-precision%29))
  (CallFunction|FindEdges SeedRT SceneCaptureZ Curl UseDepth Z%20Location)
  (Utilities|JumpFloodSteps))

(fn ComputeDistance
  (bind (_x _y) (Utilities|Struct|BreakIntPoint (Variables|Default|GetTextureSize)))
  (bind _rt_df (Variables|Default|GetRT%20DF))
  (bind _computedistancemid (Variables|Default|GetComputeDistanceMID))
  (Rendering|CreateAndAssignMID (Variables|Default|GetComputeDistanceMID) "/Landmass/DistanceFields/Materials/M_JumpFlood_Distance_2D.M_JumpFlood_Distance_2D"
    (:Out
      (Rendering|Material|SetTextureParameterValue _computedistancemid "RT" (Utilities|PingPong (Math|Integer|%25%28Integer%29 (Variables|Default|GetCompletedBlurPasses) 2)))
      (Rendering|Material|SetVectorParameterValue _computedistancemid "TextureSize" (Utilities|Struct|MakeLinearColor (Math|Conversions|ToFloat(Integer) _x) (Math|Conversions|ToFloat(Integer) _y)) :!class /Script/Engine.MaterialInstanceDynamic)
      (Rendering|ClearRenderTarget2D _rt_df)
      (Rendering|DrawMaterialtoRenderTarget _rt_df (Variables|Default|GetComputeDistanceMID))
      (Math|Integer|IncrementInt (Variables|Default|GetCompletedPasses)
        (:%20%20
          (return (Variables|Default|GetRT%20DF)))))))

(fn EnsureRenderTargetsAllocated
  (params (Seed :type Texture%20Object%20Reference))
  (bind (_x _y) (Utilities|Struct|BreakIntPoint (Variables|Default|GetTextureSize)))
  (bind (_x_2 _y_2) (Utilities|Struct|BreakIntPoint (Variables|Default|GetTextureSize)))
  (bind (_x_3 _y_3) (Utilities|Struct|BreakIntPoint (Variables|Default|GetTextureSize)))
  (Default|CheckAndCreateRT (Variables|Default|GetRTA) _x_2 _y_2 "RTF_RGBA32f"
    (:Out
      (Default|CheckAndCreateRT (Variables|Default|GetRTB) _x_3 _y_3 "RTF_RGBA32f"
        (:Out
          (Default|CheckAndCreateRT (Variables|Default|GetRT%20DF) _x _y "RTF_R32f"))))))

(fn FindEdges
  (params (SeedRT :type Texture%20Render%20Target%202D%20Object%20Reference) (CaptureZ :type Float%20%28double-precision%29) (Curl :type Linear%20Color%20Structure) (UseDepth :type Boolean) (Z%20Location :type Float%20%28double-precision%29))
  (bind _sizex (Variables|/Script/Engine.TextureRenderTarget2D|GetSizeX SeedRT))
  (bind _findedgesmid (Variables|Default|GetFindEdgesMID))
  (bind _sizey (Variables|/Script/Engine.TextureRenderTarget2D|GetSizeY SeedRT))
  (bind _findedgesmid_2 (Variables|Default|GetFindEdgesMID))
  (Variables|Default|SetPingPongOffset 0)
  (Variables|Default|SetCompletedBlurPasses 0)
  (Variables|Default|SetCompletedPasses 0)
  (Rendering|CreateAndAssignMID (Variables|Default|GetFindEdgesMID) "/Landmass/DistanceFields/Materials/DetectEdges.DetectEdges"
    (:Out
      (Variables|Default|SetTextureSize (Utilities|Struct|MakeIntPoint _sizex _sizey))
      ;[1 1 1] MID Config
      (Rendering|Material|SetVectorParameterValue _findedgesmid_2 "TextureSize" (Utilities|Struct|MakeLinearColor (Math|Conversions|ToFloat(Integer) _sizex) (Math|Conversions|ToFloat(Integer) _sizey)) :!class /Script/Engine.MaterialInstanceDynamic)
      (Rendering|Material|SetTextureParameterValue _findedgesmid_2 "SeedEdgesRT" (Variables|GetSeedRT))
      (Rendering|Material|SetVectorParameterValue _findedgesmid_2 "Curl" (Variables|GetCurl) :!class /Script/Engine.MaterialInstanceDynamic)
      (Rendering|Material|SetScalarParameterValue _findedgesmid "SceneCaptureZ" CaptureZ :!class /Script/Engine.MaterialInstanceDynamic)
      (Rendering|Material|SetScalarParameterValue _findedgesmid "UsesDepth" (Math|Conversions|ToFloat(Boolean) UseDepth) :!class /Script/Engine.MaterialInstanceDynamic)
      (Rendering|Material|SetScalarParameterValue _findedgesmid "Z" Z%20Location :!class /Script/Engine.MaterialInstanceDynamic)
      ;[1 1 1] Ensure Render Targets are Allocated
      ; Allocated them locally if needed
      (CallFunction|EnsureRenderTargetsAllocated (Variables|GetSeedRT))
      ;[1 1 1] Edge Detection Shader
      (Rendering|ClearRenderTarget2D (Variables|Default|GetRTA))
      (Rendering|ClearRenderTarget2D (Variables|Default|GetRTB))
      (Rendering|DrawMaterialtoRenderTarget (Variables|Default|GetRTA) (Variables|Default|GetFindEdgesMID))
      (return (Utilities|PingPong 0)))))

(fn FindEdges_Debug
  (params (SeedRT :type Texture%20Render%20Target%202D%20Object%20Reference) (CaptureZ :type Float%20%28double-precision%29) (Curl :type Linear%20Color%20Structure) (DestRT :type Texture%20Render%20Target%202D%20Object%20Reference) (ZOffset :type Float%20%28double-precision%29))
  (bind _findedgesmid (Variables|Default|GetFindEdgesMID))
  (bind _findedgesmid_2 (Variables|Default|GetFindEdgesMID))
  (Variables|Default|SetCompletedPasses 0)
  (Utilities|IsValid (Variables|Default|GetFindEdgesMID)
    (:Is%20Valid
      _)
    (:Is%20Not%20Valid
      (bind _returnvalue (Rendering|Material|CreateDynamicMaterialInstance "/Landmass/DistanceFields/Materials/DetectEdges.DetectEdges" "None" "None" :!class /Script/Engine.KismetMaterialLibrary))
      (Variables|Default|SetFindEdgesMID _returnvalue)))
  (Rendering|Material|SetVectorParameterValue _findedgesmid_2 "TextureSize" (Utilities|Struct|MakeLinearColor (Math|Conversions|ToFloat(Integer) (Variables|/Script/Engine.TextureRenderTarget2D|GetSizeX SeedRT)) (Math|Conversions|ToFloat(Integer) (Variables|/Script/Engine.TextureRenderTarget2D|GetSizeY SeedRT))) :!class /Script/Engine.MaterialInstanceDynamic)
  (Rendering|Material|SetTextureParameterValue _findedgesmid_2 "SeedEdgesRT" SeedRT)
  (Rendering|Material|SetVectorParameterValue _findedgesmid_2 "Curl" Curl :!class /Script/Engine.MaterialInstanceDynamic)
  (Rendering|Material|SetScalarParameterValue _findedgesmid "SceneCaptureZ" CaptureZ :!class /Script/Engine.MaterialInstanceDynamic)
  (Rendering|Material|SetScalarParameterValue _findedgesmid "UsesDepth" (select (== CaptureZ 0.0) 0.0 1.0) :!class /Script/Engine.MaterialInstanceDynamic)
  (Rendering|Material|SetScalarParameterValue _findedgesmid "Z" ZOffset :!class /Script/Engine.MaterialInstanceDynamic)
  (Rendering|ClearRenderTarget2D DestRT)
  (Rendering|DrawMaterialtoRenderTarget DestRT (Variables|Default|GetFindEdgesMID)))

(fn JumpFloodandComputeDistance
  (params (Seed :type Texture%20Render%20Target%202D%20Object%20Reference) (Curl%20Params :type Linear%20Color%20Structure) (In%20Blur%20Passes :type Integer))
  (bind _in_blur_passes (Variables|GetIn%20Blur%20Passes))
  (CallFunction|FindEdges Seed 0.0 Curl%20Params)
  (Variables|Default|SetUseBlur (> _in_blur_passes 0))
  (Variables|Default|SetBlurPasses _in_blur_passes)
  (Utilities|FlowControl|Sequence
    (:then_0
      (Utilities|JumpFloodSteps))
    (:then_1
      (bind _rt (CallFunction|ComputeDistance))
      (return _rt))))

(fn SingleBlurStep
  (bind (_x _y) (Utilities|Struct|BreakIntPoint (Variables|Default|GetTextureSize)))
  (bind (_source _target) (Utilities|PingPong 0))
  (Rendering|CreateAndAssignMID (Variables|Default|GetBlurEdgesMID) "/Landmass/DistanceFields/Materials/BlurEdges.BlurEdges"
    (:Out
      (Rendering|Material|SetVectorParameterValue (Variables|Default|GetBlurEdgesMID) "TextureResolution" (Utilities|Struct|MakeLinearColor (Math|Conversions|ToFloat(Integer) _x) (Math|Conversions|ToFloat(Integer) _y)) :!class /Script/Engine.MaterialInstanceDynamic)
      (Rendering|Material|SetTextureParameterValue (Variables|Default|GetBlurEdgesMID) "SeedEdgesRT" _source)
      (Rendering|ClearRenderTarget2D _target)
      (Rendering|DrawMaterialtoRenderTarget _target (Variables|Default|GetBlurEdgesMID))
      (Math|Integer|IncrementInt (Variables|Default|GetCompletedPasses)
        (:%20%20
          (Math|Integer|IncrementInt (Variables|Default|GetCompletedBlurPasses)
            (:%20%20
              (return (Utilities|PingPong 0)))))) ; This needs to be reset before Jump Flood begins
    )))

(fn SingleJumpStep
  (bind _completedblurpasses (Variables|Default|GetCompletedBlurPasses))
  (bind _returnvalue (Math|Integer|%25%28Integer%29 _completedblurpasses 2))
  (bind (_source _target) (Utilities|PingPong _returnvalue))
  (bind _jumpstepmid (Variables|Default|GetJumpStepMID))
  (Rendering|Material|SetScalarParameterValue _jumpstepmid "Index" (Math|Conversions|ToFloat(Integer) (+ (Variables|Default|GetCompletedPasses) 1)) :!class /Script/Engine.MaterialInstanceDynamic)
  (Rendering|Material|SetTextureParameterValue _jumpstepmid "RT" (Utilities|PingPong _returnvalue))
  (Rendering|ClearRenderTarget2D _target)
  (Rendering|DrawMaterialtoRenderTarget _target (Variables|Default|GetJumpStepMID))
  (Math|Integer|IncrementInt (Variables|Default|GetCompletedPasses)
    (:%20%20
      (return (Utilities|PingPong _returnvalue)))))
