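# Unlit Landmass material that blurs edge data in the SeedEdgesRT texture, apparently as a step in distance field generation.
# A Custom HLSL node replaces each non-empty texel's RG with the average of its non-empty eight neighbours, excluding itself, or zero if none exist.
# Empty texels pass through unchanged and BA is preserved.
# Texel size comes only from the X component of TextureResolution, so square textures are assumed.
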Material /Landmass/DistanceFields/Materials/BlurEdges
graph {
 properties {
  bAllowNegativeEmissiveColor=true
  bUseTranslucencyVertexFog=false
  blendMode=BLEND_AlphaComposite
  floatPrecisionMode=MFPM_Full
  pixelDepthOffsetMode=PDOM_Legacy
  shadingModel=MSM_Unlit
  usedShadingModels=Unlit
 }
 nodes [
  MaterialExpressionTextureObjectParameter_1 = TextureObjectParameter() {parameterName=SeedEdgesRT, texture=/Landmass/Landscape/OldPrototype_BP/Textures/LS_Debug_RT_01}
  MaterialExpressionTextureCoordinate_0 = TextureCoordinate()
  MaterialExpressionComponentMask_2 = ComponentMask(None=VectorParameter() {parameterName=TextureResolution, defaultValue=<rgba 256 256 0 0>}.RGB) {g=true, r=true}
  n1 = Custom!(Tex=@MaterialExpressionTextureObjectParameter_1, UV=@MaterialExpressionTextureCoordinate_0, TextureSize=@MaterialExpressionComponentMask_2) {
   code="float texelsize = 1/TextureSize.x;\r\n\r\nfloat2 offsets[8] = {float2(1,0) , float2(-1,0) , float2(0,1) , float2 (0,-1), float2(1,1), float2(1,-1), float2(-1,1), float2(-1,-1)};\r\n\r\nfloat4 startsample =  Tex.SampleLevel(TexSampler, UV, 0);\r\n\r\nint neighbors = 0;\r\nfloat3 accum = 0;\r\n\r\nif((startsample.r + startsample.g) == 0)\r\n{\r\n\treturn startsample;\r\n}\r\n\r\nint i = 0;\r\nwhile(i<8)\r\n{\r\n\tfloat4 offsetsample = Tex.SampleLevel(TexSampler, UV + (offsets[i] * texelsize), 0);\r\n\r\n\tif( (offsetsample.r + offsetsample.g) > 0 )\r\n\t{\r\n\t\taccum.xy += offsetsample.xy;\r\n\t\tneighbors += 1;\r\n\t}\r\n\r\ni++;\r\n}\r\n\r\nif (neighbors > 0)\r\n{\r\n\taccum.xy /= neighbors;\r\n}\r\n\r\nreturn float4( accum.xy, startsample.ba);"
   description=Blur Edges
   inputs [
    {inputName=Tex, input {expression=&self:MaterialExpressionTextureObjectParameter_1}}
    {inputName=UV, input {expression=&self:MaterialExpressionTextureCoordinate_0}}
    {inputName=TextureSize, input {expression=&self:MaterialExpressionComponentMask_2}}
   ]
  }
  Add(
   A=@n1
   B=MaterialFunctionCall(Z=ScalarParameter() {parameterName=XXXX}) {materialFunction=&MakeFloat3}.Result
  )
  n2 = TextureSampleParameter2D!(UVs, Level=Constant(), Apply View MipBias) {constMipValue=0, mipValueMode=TMVM_MipLevel, parameterName=SeedEdgesRT, texture=&LS_Debug_RT_01}
  OneMinus(None=@n2.A)
 ]
}
