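# Unlit Landmass distance-field material that finds shape edges in the SeedEdgesRT texture and encodes them as seed positions with height.
# Two curl-noise layers warp the texel-snapped lookup, and the offset is removed from the resulting edge positions.
# A Custom HLSL node compares four neighbors, optionally treating samples as capture depth, and averages half-texel edge UVs.
# Height is offset and divided by MPC_Erosion's ZScale, alongside an inverted inside-shape mask.
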Material /Landmass/DistanceFields/Materials/DetectEdges
graph {
 properties {
  bAllowNegativeEmissiveColor=true
  bUseTranslucencyVertexFog=false
  blendMode=BLEND_AlphaComposite
  floatPrecisionMode=MFPM_Full
  pixelDepthOffsetMode=PDOM_Legacy
  shadingModel=MSM_Unlit
  usedShadingModels=Unlit
 }
 nodes [
  n1 = TextureCoordinate()
  n2 = ScalarParameter() {parameterName=Seed1}
  n3 = VectorParameter() {parameterName=Curl, defaultValue=<rgba 3 0.025 24 -0.0025>}
  n4 = VectorNoise(World Position=Multiply(A=AppendVector(A=@n1, B=@n2), B=@n3.R)) {noiseFunction=VNF_CurlALU}
  n5 = Divide(A=@n3.G) {constB=100}
  n6 = TextureCoordinate()
  n7 = ScalarParameter() {parameterName=Seed2}
  n8 = VectorNoise(World Position=Multiply(A=AppendVector(A=@n6, B=@n7), B=@n3.B)) {noiseFunction=VNF_CurlALU}
  n9 = Divide(A=@n3.A) {constB=100}
  n10 = ComponentMask(None=Add(A=Multiply(A=@n4, B=@n5), B=Multiply(A=@n8, B=@n9))) {g=true, r=true}
  n11 = Multiply(A=@n10) {constB=-1}
  MaterialExpressionTextureObjectParameter_1 = TextureObjectParameter() {parameterName=SeedEdgesRT, texture=/Landmass/Landscape/OldPrototype_BP/Textures/LS_Debug_RT_01}
  MaterialExpressionComponentMask_2 = ComponentMask(None=VectorParameter() {parameterName=TextureSize, defaultValue=<rgba 256 256 0 0>}.RGB) {g=true, r=true}
  n12 = Multiply(A=Add(A=@n10, B=TextureCoordinate()), B=@MaterialExpressionComponentMask_2)
  MaterialExpressionDivide_0 = Divide(A=Add(A=Floor(None=@n12)) {constB=0.5}, B=@MaterialExpressionComponentMask_2)
  MaterialExpressionScalarParameter_0 = ScalarParameter() {defaultValue=0.1, parameterName=Thresh}
  MaterialExpressionSubtract_0 = Subtract(A=ScalarParameter() {defaultValue=500, parameterName=SceneCaptureZ}, B=ScalarParameter() {defaultValue=-50000, parameterName=MinZ})
  MaterialExpressionScalarParameter_4 = ScalarParameter() {defaultValue=1, parameterName=UsesDepth}
  n13 = Custom!(
   Tex=@MaterialExpressionTextureObjectParameter_1
   UV=@MaterialExpressionDivide_0
   TextureSize=@MaterialExpressionComponentMask_2
   Threshold=@MaterialExpressionScalarParameter_0
   SceneCaptureZ=@MaterialExpressionSubtract_0
   UsesDepth=@MaterialExpressionScalarParameter_4
  ) {
   code="float2 texelsize = 1/TextureSize;\r\n\r\nfloat2 offsets[8] = {float2(1,0) , float2(-1,0) , float2(0,1) , float2 (0,-1), float2(1,1), float2(1,-1), float2(-1,1), float2(-1,-1)};\r\n\r\nfloat startsample = Tex.SampleLevel(TexSampler, UV, 0);\r\n\r\nif(UsesDepth)\r\n{\r\n\tstartsample = SceneCaptureZ - startsample;\r\n}\r\n\r\nfloat startcheck = saturate(ceil(startsample));\r\n\r\nint neighbors = 0;\r\nfloat3 accum = 0;\r\n\r\n\r\nint i = 0;\r\nwhile(i<4)\r\n{\r\n\tfloat2 offsetUV = UV + (offsets[i] * texelsize);\r\n\t//Only wrap edges of the current edge sample was shape interior, to make sure the edge gets seeded\r\n\tif(startcheck ==0)\r\n\t{\r\n\t\toffsetUV = clamp(offsetUV, 0.5 * texelsize, 1 - (0.5 * texelsize));\r\n\t}\r\n\r\n\tfloat offsetsample = Tex.SampleLevel(TexSampler, offsetUV, 0);\r\n\tif(UsesDepth)\r\n\t{\r\n\t\toffsetsample = SceneCaptureZ - offsetsample;\r\n\t}\r\n\r\n\tfloat offsetcheck = saturate(ceil(offsetsample));\r\n\r\n\tif(offsetcheck != startcheck &&&& startcheck <= 1)\r\n\t{\r\n\r\n\t\taccum.xy += offsets[i] * texelsize * 0.5;\r\n\r\n\t\tneighbors += 1;\r\n\r\n\t\tif(offsetcheck == 1)\r\n\t\t{\r\n\t\t\taccum.b += offsetsample;\r\n\t\t}\r\n\t}\r\n\t\r\n\t\r\n\r\ni++;\r\n}\r\n\r\nif (neighbors > 0)\r\n{\r\n\taccum.xyz /= neighbors;\r\n\taccum.xy += UV;\r\n}\r\n\r\nif(startcheck == 1)\r\n{\r\n\taccum.b = startsample;\r\n}\r\n\r\n\r\nreturn accum;"
   description=Detect Edges
   inputs [
    {inputName=Tex, input {expression=&self:MaterialExpressionTextureObjectParameter_1}}
    {inputName=UV, input {expression=&self:MaterialExpressionDivide_0}}
    {inputName=TextureSize, input {expression=&self:MaterialExpressionComponentMask_2}}
    {inputName=Threshold, input {expression=&self:MaterialExpressionScalarParameter_0}}
    {inputName=SceneCaptureZ, input {expression=&self:MaterialExpressionSubtract_0}}
    {inputName=UsesDepth, input {expression=&self:MaterialExpressionScalarParameter_4}}
   ]
  }
  n14 = ComponentMask(None=@n13) {g=true, r=true}
  n15 = MaterialFunctionCall(Float2=Saturate(None=Ceil(None=@n14))) {materialFunction=&BreakOutFloat2Components}
  n16 = Multiply(A=@n15.R, B=@n15.R)
  n17 = Add(
   A=ComponentMask(None=@n13) {b=true}
   B=Multiply(A=ScalarParameter() {defaultValue=-50000, parameterName=MinZ}, B=ScalarParameter() {defaultValue=1, parameterName=UsesDepth})
  )
  n18 = ScalarParameter() {parameterName=Z}
  AppendVector(
   A=Add(A=Multiply(A=@n11, B=@n16), B=ComponentMask(None=@n13) {g=true, r=true}) {desc=Fixes Curl Noise}
   B=Divide(A=Add(A=@n17, B=@n18), B=CollectionParameter() {collection=/Landmass/Landscape/BlueprintBrushes/MPC/MPC_Erosion, parameterName=ZScale})
  )
  Constant()
  n19 = TextureSampleParameter2D!(UVs=@MaterialExpressionDivide_0, Level, Apply View MipBias) {constMipValue=0, mipValueMode=TMVM_MipLevel, parameterName=SeedEdgesRT, texture=&LS_Debug_RT_01}
  n20 = ScalarParameter() {defaultValue=500, parameterName=SceneCaptureZ}
  n21 = ScalarParameter() {defaultValue=-50000, parameterName=MinZ}
  n22 = LinearInterpolate(A=@n19.R, B=Subtract(A=Subtract(A=@n20, B=@n21), B=@n19.R), Alpha=ScalarParameter() {defaultValue=1, parameterName=UsesDepth})
  OneMinus(None=Saturate(None=Ceil(None=@n22)))
 ]
}
