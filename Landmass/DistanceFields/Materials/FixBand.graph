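# FixBand is an unlit, alpha-composite material in the Landmass distance field set, allowing negative emissive values.
# It samples the SeedEdgesRT texture parameter at mip 0, defaulting to a debug render target from an old landscape prototype.
# The sample's alpha feeds a OneMinus node to invert it.
# The graph shows no connection from that result to a material output.
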Material /Landmass/DistanceFields/Materials/FixBand
graph {
 properties {
  bAllowNegativeEmissiveColor=true
  blendMode=BLEND_AlphaComposite
  pixelDepthOffsetMode=PDOM_Legacy
  shadingModel=MSM_Unlit
  usedShadingModels=Unlit
 }
 nodes [
  n1 = TextureSampleParameter2D!(UVs, Level, Apply View MipBias) {constMipValue=0, mipValueMode=TMVM_MipLevel, parameterName=SeedEdgesRT, texture=/Landmass/Landscape/OldPrototype_BP/Textures/LS_Debug_RT_01}
  OneMinus(None=@n1.A)
 ]
}
