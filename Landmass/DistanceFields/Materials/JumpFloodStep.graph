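# Unlit material for one jump flood pass in Landmass distance field generation, propagating nearest-seed coordinates across the RT texture.
# A Custom node samples eight neighbours around each texel centre, keeping the seed UV in RG closest in UV space and carrying B along.
# Step size is two to the power of ceil(log2(TextureSize.x)) minus Index, in pixels.
# Landscape world size feeds the node but does not affect its result.
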Material /Landmass/DistanceFields/Materials/JumpFloodStep
graph {
 properties {
  bAllowNegativeEmissiveColor=true
  bUseTranslucencyVertexFog=false
  blendMode=BLEND_AlphaComposite
  floatPrecisionMode=MFPM_Full
  pixelDepthOffsetMode=PDOM_Legacy
  shadingModel=MSM_Unlit
  usedShadingModels=Unlit
 }
 nodes [
  MaterialExpressionTextureObjectParameter_0 = TextureObjectParameter() {parameterName=RT, texture=&Black}
  n1 = ComponentMask(None=VectorParameter() {parameterName=TextureSize, defaultValue=<rgba 512 512 0 0>}.RGB) {g=true, r=true}
  n2 = Multiply(A=TextureCoordinate(), B=@n1)
  MaterialExpressionAdd_3 = Add(A=Divide(A=Floor(None=@n2), B=@n1), B=Divide(A=Divide(B=@n1)))
  n3 = VectorParameter() {parameterName=TextureSize, defaultValue=<rgba 512 512 0 0>}
  MaterialExpressionComponentMask_9 = ComponentMask(None=@n3.RGB) {g=true, r=true}
  n4 = Logarithm2!(X=Max(A=@n3.R, B=@n3.R) {desc=Size})
  MaterialExpressionPower_0 = Power(Base=Constant() {r=2}, Exp=Subtract(A=Ceil(None=@n4), B=ScalarParameter() {defaultValue=-1, parameterName=Index}))
  MaterialExpressionAppendVector_1 = AppendVector(
   A=CollectionParameter() {collection=/Landmass/Landscape/BlueprintBrushes/MPC/MPC_Landscape, parameterName=WorldSizeX}
   B=CollectionParameter() {collection=&MPC_Landscape, parameterName=WorldSizeY}
  )
  n5 = Custom!(
   Tex=@MaterialExpressionTextureObjectParameter_0
   UV0=@MaterialExpressionAdd_3
   TextureSize=@MaterialExpressionComponentMask_9
   offsetmult=@MaterialExpressionPower_0
   WorldSize=@MaterialExpressionAppendVector_1
  ) {
   code="float2 PixelSize = 1 / TextureSize;\r\nfloat2 offsets[8] = {float2(1,0) , float2(-1,0) , float2(0,1) , float2 (0,-1), float2(1,1), float2(1,-1), float2(-1,1), float2(-1,-1)};\r\n\r\nfloat2 simpleoffsets[4] = {float2(1,0), float2(-1,0), float2(0,1), float2(0,-1)};\r\nfloat3 UV3d = float3(UV0, 0) * float3(WorldSize, 1);\r\nfloat3 cursample = Tex.SampleLevel(TexSampler, UV0, 0);\r\n\r\n\r\nfloat curlen = length( cursample.xy - UV0 );\r\n\r\nfloat minlen = curlen;\r\nif (cursample.x == 0 &&&& cursample.y == 0) minlen = 10000000;\r\n\r\nfloat3 nearest = cursample.rgb;\r\n\r\nfor (int i = 0; i < 8; i++)\r\n{\r\n\r\n\t//if (offsetmult < 1) \r\n\t//{\r\n\t//\treturn  Tex.SampleLevel(TexSampler, UV0, 0).rgb;\r\n\t//}\r\n\t\r\n\tfloat2 CurUV = UV0 + offsets[i] * offsetmult * PixelSize;\r\n\r\n\tcursample = Tex.SampleLevel(TexSampler, CurUV, 0).rgb;\r\n\r\n\r\n\tif ( (cursample.x > 0 || cursample.y > 0) &&&& CurUV.x > 0 &&&& CurUV.y > 0 &&&& CurUV.x < 1 &&&& CurUV.y < 1)\r\n\t{\r\n\t\t//cursample.xy += floor(CurUV);\r\n\t\t\r\n\t\tcurlen = length( cursample.xy - UV0 );\r\n\t\tif ( curlen < minlen )\r\n\t\t{\r\n\t\t\tminlen = curlen;\r\n\t\t\tnearest = cursample.rgb;\r\n\r\n\t\t}\r\n\t}\r\n\r\n}\r\n\r\nreturn nearest;"
   description=Jump Step
   inputs [
    {inputName=Tex, input {expression=&self:MaterialExpressionTextureObjectParameter_0}}
    {inputName=UV0, input {expression=&self:MaterialExpressionAdd_3}}
    {inputName=TextureSize, input {expression=&self:MaterialExpressionComponentMask_9}}
    {inputName=offsetmult, input {expression=&self:MaterialExpressionPower_0}}
    {inputName=WorldSize, input {expression=&self:MaterialExpressionAppendVector_1}}
   ]
  }
  n6 = ComponentMask(None=@n5) {b=true}
  AppendVector(A=ComponentMask(None=@n5) {g=true, r=true}, B=@n6)
  n7 = TextureSampleParameter2D(UVs=@MaterialExpressionAdd_3) {parameterName=RT, texture=&Black}
  LinearInterpolate(A=@n6, B=@n7.B, Alpha=@n7.B)
  OneMinus(None=@n7.A)
  VectorParameter()
 ]
}
