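# Unlit material that converts a 2D jump flood result in the RT texture parameter into a signed distance field for Landmass.
# It samples RT at point-snapped UVs, reads RG as nearest-seed coordinates and measures the UV-space distance to the current texel.
# Alpha selects the sign through a lerp, and a bias-scale remaps the signed distance, with full float precision.
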Material /Landmass/DistanceFields/Materials/M_JumpFlood_Distance_2D
graph {
 properties {
  bAllowNegativeEmissiveColor=true
  floatPrecisionMode=MFPM_Full
  pixelDepthOffsetMode=PDOM_Legacy
  shadingModel=MSM_Unlit
  usedShadingModels=Unlit
 }
 nodes [
  n1 = MaterialFunctionCall!(
   UVs=TextureCoordinate()
   Texture Size=TextureProperty(Tex=TextureObjectParameter() {parameterName=RT, texture=&Black})
  ) {materialFunction=/Engine/Functions/Engine_MaterialFunctions01/Texturing/PointSampledUVs}
  n2 = TextureSampleParameter2D!(UVs=@n1.Result, Level=Constant(), Apply View MipBias) {
   mipValueMode=TMVM_MipLevel
   parameterName=RT
   samplerType=SAMPLERTYPE_LinearColor
   texture=/Engine/Functions/Engine_MaterialFunctions02/PivotPainter2/Black_1x1_EXR_Texture
  }
  n3 = ComponentMask(None=@n2.RGB) {g=true, r=true}
  ConstantBiasScale(None=Multiply(A=Distance(A=@n1.Result, B=@n3), B=LinearInterpolate(Alpha=@n2.A) {constA=-1}))
 ]
}
